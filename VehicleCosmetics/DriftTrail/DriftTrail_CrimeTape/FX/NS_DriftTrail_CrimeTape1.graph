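# Niagara drift-trail vehicle cosmetic that streams crime-tape ribbons behind the car and wraps two rear wheels in spinning meshes.
# Warning, DoNot and Tape emitters each spawn two skeletal-attached ribbons, mirrored by spawn group, with game-mode-dependent motion or lifetime.
# Age-driven sines and curves make ribbons flutter, twist and fade between yellow and black color parameters.
# Tire-FX masking gates each side; painted-variant overrides are present but disabled.
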NiagaraSystem /VehicleCosmetics/DriftTrail/DriftTrail_CrimeTape/FX/NS_DriftTrail_CrimeTape1
dependencies {
 usedDataInterfaces [&NiagaraDataInterfaceArrayFloat3, &NiagaraDataInterfaceCurve, &NiagaraDataInterfaceSkeletalMesh]
 usedDynamicInputs []
 usedModules [
  &self:SystemSpawnScript.SystemScriptSource.SystemScriptGraph.NiagaraNodeAssignment_0.SetVariables_#1
  &FN_UnpackVehicleFlags
  /VehicleCosmetics/FX/Niagara/Modules/VC_SetVehicleSpeedAttribs
  &SystemState
 ]
}
emitters {
 NE_R1_Warning001 {
  emitterUpdateScript [
   &EmitterState()
   &VC_RateQualityHandler(Epic Spawn Rate=30, High Spawn Rate=30, Low Spawn Rate=30, Use Custom Rate?=true)
   &VC_SpawnRibbon_v1(Number of Ribbons=2, SpawnRate=$Emitter.Spawn Rate)
  ]
  particleSpawnScript [
   &VC_PaintedVariantsOverride(
    Black=<rgba 0 0 0 1>
    BurntSienna=<rgba 0.3333 0.1029 0.0411 1>
    Cobalt=<rgba 0.0117 0.0332 1.1719 1>
    Crimson=<rgba 0.75 0.0547 0.0619 1>
    DefaultColor=<rgba 0.8308 0.5333 0 1>
    ForestGreen=<rgba 0 0.5 0.0253 1>
    Grey=<rgba 0.3125 0.3125 0.3125 1>
    Lime=<rgba 0.3828 0.7656 0 1>
    Orange=<rgba 1 0.2059 0 1>
    Pink=<rgba 0.7865 0.172 0.563 1>
    Purple=<rgba 0.1452 0.0508 0.8125 1>
    R1 Black:LinearColor=<rgba 0.2031 0.1998 0.1881 1>
    R1 BurntSienna:LinearColor=<rgba 0.0417 0.0129 0.0051 1>
    R1 Cobalt:LinearColor=<rgba 0 0.004 0.2135 1>
    R1 Crimson:LinearColor=<rgba 0.0833 0.0004 0 1>
    R1 DefaultColor=<rgba 0 0 0 1>
    R1 ForestGreen:LinearColor=<rgba 0 0.0625 0.0032 1>
    R1 Gold:LinearColor=<rgba 15 15 15 1>
    R1 Grey:LinearColor=<rgba 0.0573 0.0573 0.0573 1>
    R1 Lime:LinearColor=<rgba 0.0417 0.0833 0 1>
    R1 Orange:LinearColor=<rgba 0.151 0.0311 0 1>
    R1 Pink:LinearColor=<rgba 0.1042 0 0.0663 1>
    R1 Purple:LinearColor=<rgba 0.0239 0 0.1927 1>
    R1 Saffron:LinearColor=<rgba 0.1146 0.0878 0 1>
    R1 SkyBlue:LinearColor=<rgba 0 0.0421 0.1146 1>
    R1 TitaniumWhite:LinearColor=<rgba 0.1042 0.1042 0.1042 1>
    Saffron=<rgba 0.8542 0.6725 0.0756 1>
    SkyBlue=<rgba 0.0469 0.397 1 1>
    TitaniumWhite=<rgba 0.7083 0.7083 0.7083 1>
    Use R1 Painted Overrides=true
   ) {enabled=false}
   &VC_InitializeRibbon_v1(
    Custom Offset=true
    Enable Skeletal Position=true
    Facing Axis:ENiagaraOrientationAxis=::YAxis
    Lifetime=&MergeHelper_FloatSwitch(0=1, 1=1, 2=0.5, 3=?, Index=$System.GameModeID, Num=::NewEnumerator1)
    Local Vector Offset=&AddVectorToPosition(
     Position=&MergeHelper_VectorSwitch(0=<5 0 -38>, 1=<5 0 -38>, 2=<8 0 -38>, 3=?, Index=$System.GameModeID, Num=::NewEnumerator1)
     Vector=$User.offset
    )
    Offset Amount=1
    Ribbon Width=10
   )
   &AddVelocity(Velocity=&MergeHelper_VectorSwitch(0=<-900 0 0>, 1=<-900 0 0>, 2=<-1000 0 0>, 3=?, Index=$System.GameModeID, Num=::NewEnumerator1)) {module=AddVelocity001}
   /SetVariables() {isSetParameters=true}
   /SetVariables(Particles.Position=&AddVectorToPosition(
     Position=$Particles.Position
     Vector=&TransformVector(
      Vector=&MakeVector(X=0, Y=&MaskFloatBySpawnGroup(Masked Float=10, Passthrough Float=-10), Z=0)
      Destination Space=::Simulation
     )
    )) {isSetParameters=true, module=SetVariables_#2}
  ]
  particleUpdateScript [
   &FN_DisableTireFX(Mask Index Attribute:NiagaraInt32=$Particles.SpawnGroup, Mask Index FL=2, Mask Index FR=3, Mask Index RL=0, Mask Index RR=1)
   /SetVariables(Particles.RibbonFacing=&MakeVector(
     X=0
     Y=&Sine(
      Normalized Angle=&Subtract_Float(A=$Emitter.NormalizedLoopAge, B=$Particles.NormalizedAge)
      Period=1
      Scale=&FloatFromCurve(FloatCurve=DataInterface() {propertyValues {Curve {
          defaultValue=1
          keys [
           {arriveTangent=4.1785, interpMode=RCIM_Cubic, leaveTangent=4.1785, tangentMode=RCTM_User, time=0, value=0}
           {arriveTangent=0.0224, interpMode=RCIM_Cubic, leaveTangent=0.0224, tangentMode=RCTM_User, time=0.5, value=1}
          ]
         }}})
      Bias=0
     )
     Z=1
    )) {enabled=false, isSetParameters=true, module=SetVariables_#3}
   &InheritSourceMovement(
    Applied Position Scale=&Multiply_VectorByFloat(Vector=<1 1 1>, Float=1)
    ApplyForce=false
    ApplyPosition=true
    ApplyPositionFalloff=false
   )
   &ParticleState()
   &SolveForcesAndVelocity(Clamp Velocity=false, Limit Acceleration=false)
   &ScaleColor(ScaleA=true, ScaleRGB=true)
   &VC_UpdateRibbon_v1(Detach Locked Particle=true, Enable Skeletal Position=true)
   /SetVariables(
    Particles.BackGround:LinearColor=&MultiplyLinearColor(
     Linear Color=<rgba 0.831 0.5334 0 55>
     Scale Factor=&MakeVector4FromVectorAndFloat(
      "Vector (RGB)"=<1 1 1>
      "Float (Alpha)"=&FloatFromCurve(FloatCurve=DataInterface() {propertyValues {Curve {
          defaultValue=-0.9281
          keys [<at 0 0 RCIM_Cubic>, {arriveTangent=0.0361, interpMode=RCIM_Cubic, leaveTangent=0.0361, tangentMode=RCTM_User, time=0.8, value=1}]
         }}})
     )
    )
    Particles.BlackColor:LinearColor=&MultiplyLinearColor(
     Linear Color=<rgba 0 0 0 1>
     Scale Factor=&MakeVector4FromVectorAndFloat(
      "Vector (RGB)"=<1 1 1>
      "Float (Alpha)"=&FloatFromCurve(FloatCurve=DataInterface() {propertyValues {Curve {keys [
           <at 0 -1 RCIM_Cubic>
           {arriveTangent=6.078, interpMode=RCIM_Cubic, leaveTangent=6.078, tangentMode=RCTM_User, time=0.2, value=3}
           {arriveTangent=-5.2659, interpMode=RCIM_Cubic, leaveTangent=-5.2659, tangentMode=RCTM_User, time=0.78, value=5}
           <at 1 -1.5 RCIM_Cubic>
          ]}}})
     )
    )
   ) {isSetParameters=true, module=SetVariables_#4}
   /SetVariables(Particles.Position=&AddVectorToPosition(
     Position=$Particles.Position
     Vector=&TransformVector(
      Vector=&MakeVector(
       X=0
       Y=0
       Z=&Sine(
        Normalized Angle=&Add_Float(
         A=&Multiply_Float(A=$Particles.NormalizedAge, B=-1.3)
         B=&Multiply_Float(A=$Emitter.LoopedAge, B=1)
        )
        Period=1
        Scale=&FloatFromCurve(FloatCurve=DataInterface() {propertyValues {Curve {keys [<at 0.1 0>,<at 0.5 1>]}}}, Scale Curve=4)
        Bias=0
       )
      )
      Destination Space=::Simulation
      Apply Scale=true
     )
    )) {isSetParameters=true, module=SetVariables_#5}
   /SetVariables(Particles.Position=&AddVectorToPosition(
     Position=$Particles.Position
     Vector=/VehicleCosmetics/FX/Niagara/Modules/VC_ConvertVectorToWorld(Vector:Vector3f=&MakeVector(
       X=0
       Y=0
       Z=&FloatFromCurve(
        FloatCurve=DataInterface() {propertyValues {Curve {keys [
            <at 0 0 RCIM_Cubic>
            {arriveTangent=17.5765, interpMode=RCIM_Cubic, leaveTangent=17.5765, tangentMode=RCTM_User, time=0.2, value=10}
            {arriveTangent=-16.7908, interpMode=RCIM_Cubic, leaveTangent=-16.7908, tangentMode=RCTM_User, time=0.8, value=11}
            <at 1 0 RCIM_Cubic>
           ]}}}
        Scale Curve=0.2
       )
      ))
    )) {isSetParameters=true, module=SetVariables_#6}
   /SetVariables(Particles.RibbonFacing=&MaskVectorBySpawnGroup(
     Masked Vector=&MakeVector(
      X=1
      Y=0
      Z=&Sine(
       Normalized Angle=&Add_Float(
        A=&Multiply_Float(A=$Particles.NormalizedAge, B=1)
        B=&Multiply_Float(A=$Emitter.LoopedAge, B=1)
       )
       Period=1
       Scale=&FloatFromCurve(FloatCurve=DataInterface() {propertyValues {Curve {
           defaultValue=1
           keys [
            {arriveTangent=4.1785, interpMode=RCIM_Cubic, leaveTangent=4.1785, tangentMode=RCTM_User, time=0, value=0}
            {arriveTangent=0.0224, interpMode=RCIM_Cubic, leaveTangent=0.0224, tangentMode=RCTM_User, time=0.5, value=1}
           ]
          }}})
       Bias=0
      )
     )
     Passthrough Vector=&MakeVector(
      X=1
      Y=0
      Z=&Sine(
       Normalized Angle=&Add_Float(
        A=&Multiply_Float(A=$Particles.NormalizedAge, B=1)
        B=&Multiply_Float(A=$Emitter.LoopedAge, B=1)
       )
       Period=1
       Scale=&FloatFromCurve(FloatCurve=DataInterface() {propertyValues {Curve {
           defaultValue=1
           keys [
            {arriveTangent=4.1785, interpMode=RCIM_Cubic, leaveTangent=4.1785, tangentMode=RCTM_User, time=0, value=0}
            {arriveTangent=0.0224, interpMode=RCIM_Cubic, leaveTangent=0.0224, tangentMode=RCTM_User, time=0.5, value=1}
           ]
          }}})
       Bias=0
      )
     )
    )) {isSetParameters=true, module=SetVariables_#7}
   /SetVariables(
    Particles.R1Width:NiagaraFloat=&SelectFloatFromArray(Direct Array Index=&SpawnGroup(), Float Selection Array=DataInterface() {propertyValues {FloatData [0,-10]}})
    Particles.R2Width:NiagaraFloat=&SelectFloatFromArray(Direct Array Index=&SpawnGroup(), Float Selection Array=DataInterface() {propertyValues {FloatData [10,0]}})
   ) {isSetParameters=true}
   &ScaleRibbonWidth(Ribbon Width Scale=&SelectFloatFromArray(Direct Array Index=&SpawnGroup(), Float Selection Array=DataInterface() {propertyValues {FloatData [1,-1]}}))
  ]
  rendererClasses [&NiagaraRibbonRendererProperties]
  renderers [
   {rendererClass=&NiagaraRibbonRendererProperties}
   {rendererClass=&NiagaraRibbonRendererProperties, rendererIndex=1}
   {rendererClass=&NiagaraRibbonRendererProperties, rendererIndex=2}
  ]
 }
 NE_R2_DoNot001 {
  emitterUpdateScript [
   &EmitterState()
   &VC_RateQualityHandler(Epic Spawn Rate=30, High Spawn Rate=30, Low Spawn Rate=30, Use Custom Rate?=true)
   &VC_SpawnRibbon_v1(Number of Ribbons=2, SpawnRate=$Emitter.Spawn Rate)
  ]
  particleSpawnScript [
   &VC_PaintedVariantsOverride(
    Black=<rgba 0 0 0 1>
    BurntSienna=<rgba 0.3333 0.1029 0.0411 1>
    Cobalt=<rgba 0.0117 0.0332 1.1719 1>
    Crimson=<rgba 0.75 0.0547 0.0619 1>
    DefaultColor=<rgba 0.8308 0.5333 0 1>
    ForestGreen=<rgba 0 0.5 0.0253 1>
    Grey=<rgba 0.3125 0.3125 0.3125 1>
    Lime=<rgba 0.3828 0.7656 0 1>
    Orange=<rgba 1 0.2059 0 1>
    Pink=<rgba 0.7865 0.172 0.563 1>
    Purple=<rgba 0.1452 0.0508 0.8125 1>
    R1 Black:LinearColor=<rgba 0.2031 0.1998 0.1881 1>
    R1 BurntSienna:LinearColor=<rgba 0.0417 0.0129 0.0051 1>
    R1 Cobalt:LinearColor=<rgba 0 0.004 0.2135 1>
    R1 Crimson:LinearColor=<rgba 0.0833 0.0004 0 1>
    R1 DefaultColor=<rgba 0 0 0 1>
    R1 ForestGreen:LinearColor=<rgba 0 0.0625 0.0032 1>
    R1 Gold:LinearColor=<rgba 15 15 15 1>
    R1 Grey:LinearColor=<rgba 0.0573 0.0573 0.0573 1>
    R1 Lime:LinearColor=<rgba 0.0417 0.0833 0 1>
    R1 Orange:LinearColor=<rgba 0.151 0.0311 0 1>
    R1 Pink:LinearColor=<rgba 0.1042 0 0.0663 1>
    R1 Purple:LinearColor=<rgba 0.0239 0 0.1927 1>
    R1 Saffron:LinearColor=<rgba 0.1146 0.0878 0 1>
    R1 SkyBlue:LinearColor=<rgba 0 0.0421 0.1146 1>
    R1 TitaniumWhite:LinearColor=<rgba 0.1042 0.1042 0.1042 1>
    Saffron=<rgba 0.8542 0.6725 0.0756 1>
    SkyBlue=<rgba 0.0469 0.397 1 1>
    TitaniumWhite=<rgba 0.7083 0.7083 0.7083 1>
    Use R1 Painted Overrides=true
   ) {enabled=false}
   &VC_InitializeRibbon_v1(
    Custom Offset=true
    Enable Skeletal Position=true
    Facing Axis:ENiagaraOrientationAxis=::YAxis
    Lifetime=&MergeHelper_FloatSwitch(0=1, 1=1, 2=0.6, 3=?, Index=$System.GameModeID, Num=::NewEnumerator1)
    Local Vector Offset=&AddVectorToPosition(
     Position=&MergeHelper_VectorSwitch(0=<5 0 -38>, 1=<5 0 -38>, 2=<8 0 -38>, 3=?, Index=$System.GameModeID, Num=::NewEnumerator1)
     Vector=$User.offset
    )
    Offset Amount=1
    Ribbon Width=10
   )
   &AddVelocity(Velocity=&MergeHelper_VectorSwitch(0=<-950 0 0>, 1=<-950 0 0>, 2=<-1000 0 0>, 3=?, Index=$System.GameModeID, Num=::NewEnumerator1)) {module=AddVelocity001}
   /SetVariables() {isSetParameters=true}
   /SetVariables(Particles.Position=&AddVectorToPosition(
     Position=$Particles.Position
     Vector=&TransformVector(
      Vector=&MakeVector(X=0, Y=&MaskFloatBySpawnGroup(Masked Float=-0.8, Passthrough Float=0.8), Z=0)
      Destination Space=::Simulation
     )
    )) {isSetParameters=true, module=SetVariables_#2}
  ]
  particleUpdateScript [
   &FN_DisableTireFX(Mask Index Attribute:NiagaraInt32=$Particles.SpawnGroup, Mask Index FL=2, Mask Index FR=3, Mask Index RL=0, Mask Index RR=1)
   /SetVariables(Particles.RibbonFacing=&MakeVector(
     X=0
     Y=&Sine(
      Normalized Angle=&Subtract_Float(A=$Emitter.NormalizedLoopAge, B=$Particles.NormalizedAge)
      Period=1
      Scale=&FloatFromCurve(FloatCurve=DataInterface() {propertyValues {Curve {
          defaultValue=1
          keys [
           {arriveTangent=4.1785, interpMode=RCIM_Cubic, leaveTangent=4.1785, tangentMode=RCTM_User, time=0, value=0}
           {arriveTangent=0.0224, interpMode=RCIM_Cubic, leaveTangent=0.0224, tangentMode=RCTM_User, time=0.5, value=1}
          ]
         }}})
      Bias=0
     )
     Z=1
    )) {enabled=false, isSetParameters=true, module=SetVariables_#3}
   &InheritSourceMovement(
    Applied Position Scale=&Multiply_VectorByFloat(Vector=<1 1 1>, Float=1)
    ApplyForce=false
    ApplyPosition=true
    ApplyPositionFalloff=false
   )
   &ParticleState()
   &SolveForcesAndVelocity(Clamp Velocity=false, Limit Acceleration=false)
   &ScaleColor(ScaleA=true, ScaleRGB=true)
   &VC_UpdateRibbon_v1(Detach Locked Particle=true, Enable Skeletal Position=true)
   /SetVariables(
    Particles.BackGround01:LinearColor=&MultiplyLinearColor(
     Linear Color=<rgba 0.831 0.5334 0 55>
     Scale Factor=&MakeVector4FromVectorAndFloat(
      "Vector (RGB)"=<1 1 1>
      "Float (Alpha)"=&FloatFromCurve(FloatCurve=DataInterface() {propertyValues {Curve {
          defaultValue=-0.9281
          keys [<at 0 0 RCIM_Cubic>, {arriveTangent=0.0361, interpMode=RCIM_Cubic, leaveTangent=0.0361, tangentMode=RCTM_User, time=0.8, value=1}]
         }}})
     )
    )
    Particles.BlackColor:LinearColor=&MultiplyLinearColor(
     Linear Color=<rgba 0 0 0 1>
     Scale Factor=&MakeVector4FromVectorAndFloat(
      "Vector (RGB)"=<1 1 1>
      "Float (Alpha)"=&FloatFromCurve(FloatCurve=DataInterface() {propertyValues {Curve {keys [
           {arriveTangent=-28.5714, interpMode=RCIM_Cubic, time=0, value=-1}
           {arriveTangent=15.8531, interpMode=RCIM_Cubic, leaveTangent=15.8531, tangentMode=RCTM_User, time=0.19, value=2}
           {arriveTangent=-0.0258, interpMode=RCIM_Cubic, leaveTangent=-0.0258, tangentMode=RCTM_User, time=0.79, value=5}
           <at 1 -1.5 RCIM_Cubic>
          ]}}})
     )
    )
   ) {isSetParameters=true, module=SetVariables_#4}
   /SetVariables(Particles.Position=&AddVectorToPosition(
     Position=$Particles.Position
     Vector=&TransformVector(
      Vector=&MakeVector(
       X=0
       Y=0
       Z=&Sine(
        Normalized Angle=&Add_Float(
         A=&Multiply_Float(A=$Particles.NormalizedAge, B=-1.5)
         B=&Multiply_Float(A=$Emitter.LoopedAge, B=1)
        )
        Period=1
        Scale=&FloatFromCurve(FloatCurve=DataInterface() {propertyValues {Curve {keys [<at 0 0>,<at 0.5 1.2>]}}}, Scale Curve=4.5)
        Bias=0
       )
      )
      Destination Space=::Simulation
      Apply Scale=true
     )
    )) {isSetParameters=true, module=SetVariables_#5}
   /SetVariables(Particles.Position=&AddVectorToPosition(
     Position=$Particles.Position
     Vector=&VC_ConvertVectorToWorld(Vector:Vector3f=&MakeVector(
       X=0
       Y=0
       Z=&FloatFromCurve(
        FloatCurve=DataInterface() {propertyValues {Curve {keys [
            <at 0 5 RCIM_Cubic>
            {arriveTangent=22.1718, interpMode=RCIM_Cubic, leaveTangent=22.1718, tangentMode=RCTM_User, time=0.35, value=18}
            {arriveTangent=-4.504, interpMode=RCIM_Cubic, leaveTangent=-4.504, tangentMode=RCTM_User, time=0.6, value=15}
            {arriveTangent=0.7473, interpMode=RCIM_Cubic, leaveTangent=0.7473, tangentMode=RCTM_User, time=1, value=5}
           ]}}}
        Scale Curve=0.2
       )
      ))
    )) {isSetParameters=true, module=SetVariables_#6}
   /SetVariables(Particles.RibbonFacing=&MaskVectorBySpawnGroup(
     Masked Vector=&MakeVector(
      X=1
      Y=0
      Z=&Sine(
       Normalized Angle=&Add_Float(
        A=&Multiply_Float(A=$Particles.NormalizedAge, B=1)
        B=&Multiply_Float(A=$Emitter.LoopedAge, B=1)
       )
       Period=1
       Scale=&FloatFromCurve(FloatCurve=DataInterface() {propertyValues {Curve {
           defaultValue=1
           keys [
            {arriveTangent=4.1785, interpMode=RCIM_Cubic, leaveTangent=4.1785, tangentMode=RCTM_User, time=0, value=0}
            {arriveTangent=0.0224, interpMode=RCIM_Cubic, leaveTangent=0.0224, tangentMode=RCTM_User, time=0.5, value=1}
           ]
          }}})
       Bias=0
      )
     )
     Passthrough Vector=&MakeVector(
      X=1
      Y=0
      Z=&Sine(
       Normalized Angle=&Add_Float(
        A=&Multiply_Float(A=$Particles.NormalizedAge, B=1)
        B=&Multiply_Float(A=$Emitter.LoopedAge, B=1)
       )
       Period=1
       Scale=&FloatFromCurve(FloatCurve=DataInterface() {propertyValues {Curve {
           defaultValue=1
           keys [
            {arriveTangent=4.1785, interpMode=RCIM_Cubic, leaveTangent=4.1785, tangentMode=RCTM_User, time=0, value=0}
            {arriveTangent=0.0224, interpMode=RCIM_Cubic, leaveTangent=0.0224, tangentMode=RCTM_User, time=0.5, value=1}
           ]
          }}})
       Bias=0
      )
     )
    )) {isSetParameters=true, module=SetVariables_#7}
   /SetVariables(
    Particles.R2Width:NiagaraFloat=&SelectFloatFromArray(Direct Array Index=&SpawnGroup(), Float Selection Array=DataInterface() {propertyValues {FloatData [0,-10]}})
    Particles.R3Width:NiagaraFloat=&SelectFloatFromArray(Direct Array Index=&SpawnGroup(), Float Selection Array=DataInterface() {propertyValues {FloatData [10,0]}})
   ) {isSetParameters=true}
   &ScaleRibbonWidth(Ribbon Width Scale=&SelectFloatFromArray(Direct Array Index=&SpawnGroup(), Float Selection Array=DataInterface() {propertyValues {FloatData [1,-1]}}))
  ]
  rendererClasses [&NiagaraRibbonRendererProperties]
  renderers [
   {rendererClass=&NiagaraRibbonRendererProperties}
   {rendererClass=&NiagaraRibbonRendererProperties, rendererIndex=1}
   {rendererClass=&NiagaraRibbonRendererProperties, rendererIndex=2}
  ]
 }
 NE_R3_Tape001 {
  emitterUpdateScript [
   &EmitterState()
   &VC_RateQualityHandler(Epic Spawn Rate=30, High Spawn Rate=30, Low Spawn Rate=30, Use Custom Rate?=true)
   &VC_SpawnRibbon_v1(Number of Ribbons=2, SpawnRate=$Emitter.Spawn Rate)
  ]
  particleSpawnScript [
   &VC_PaintedVariantsOverride(
    Black=<rgba 0 0 0 1>
    BurntSienna=<rgba 0.3333 0.1029 0.0411 1>
    Cobalt=<rgba 0.0117 0.0332 1.1719 1>
    Crimson=<rgba 0.75 0.0547 0.0619 1>
    DefaultColor=<rgba 0.8308 0.5333 0 1>
    ForestGreen=<rgba 0 0.5 0.0253 1>
    Grey=<rgba 0.3125 0.3125 0.3125 1>
    Lime=<rgba 0.3828 0.7656 0 1>
    Orange=<rgba 1 0.2059 0 1>
    Pink=<rgba 0.7865 0.172 0.563 1>
    Purple=<rgba 0.1452 0.0508 0.8125 1>
    R1 Black:LinearColor=<rgba 0.2031 0.1998 0.1881 1>
    R1 BurntSienna:LinearColor=<rgba 0.0417 0.0129 0.0051 1>
    R1 Cobalt:LinearColor=<rgba 0 0.004 0.2135 1>
    R1 Crimson:LinearColor=<rgba 0.0833 0.0004 0 1>
    R1 DefaultColor=<rgba 0 0 0 1>
    R1 ForestGreen:LinearColor=<rgba 0 0.0625 0.0032 1>
    R1 Gold:LinearColor=<rgba 15 15 15 1>
    R1 Grey:LinearColor=<rgba 0.0573 0.0573 0.0573 1>
    R1 Lime:LinearColor=<rgba 0.0417 0.0833 0 1>
    R1 Orange:LinearColor=<rgba 0.151 0.0311 0 1>
    R1 Pink:LinearColor=<rgba 0.1042 0 0.0663 1>
    R1 Purple:LinearColor=<rgba 0.0239 0 0.1927 1>
    R1 Saffron:LinearColor=<rgba 0.1146 0.0878 0 1>
    R1 SkyBlue:LinearColor=<rgba 0 0.0421 0.1146 1>
    R1 TitaniumWhite:LinearColor=<rgba 0.1354 0.1354 0.1354 1>
    Saffron=<rgba 0.8542 0.6725 0.0756 1>
    SkyBlue=<rgba 0.0469 0.397 1 1>
    TitaniumWhite=<rgba 0.7083 0.7083 0.7083 1>
    Use R1 Painted Overrides=true
   ) {enabled=false}
   &VC_InitializeRibbon_v1(
    Custom Offset=true
    Enable Skeletal Position=true
    Facing Axis:ENiagaraOrientationAxis=::YAxis
    Lifetime=&MergeHelper_FloatSwitch(0=1, 1=1, 2=0.7, 3=?, Index=$System.GameModeID, Num=::NewEnumerator1)
    Local Vector Offset=&AddVectorToPosition(
     Position=&MergeHelper_VectorSwitch(0=<5 0 -38>, 1=<5 0 -38>, 2=<8 0 -38>, 3=?, Index=$System.GameModeID, Num=::NewEnumerator1)
     Vector=$User.offset
    )
    Offset Amount=1
    Ribbon Width=10
   )
   &AddVelocity(Velocity=&MergeHelper_VectorSwitch(0=<-1000 0 0>, 1=<-1000 0 0>, 2=<-1000 0 0>, 3=?, Index=$System.GameModeID, Num=::NewEnumerator1)) {module=AddVelocity001}
   /SetVariables() {isSetParameters=true}
   /SetVariables(Particles.Position=&AddVectorToPosition(
     Position=$Particles.Position
     Vector=&TransformVector(
      Vector=&MakeVector(X=0, Y=&MaskFloatBySpawnGroup(Masked Float=-10, Passthrough Float=10), Z=0)
      Destination Space=::Simulation
     )
    )) {isSetParameters=true, module=SetVariables_#2}
  ]
  particleUpdateScript [
   &FN_DisableTireFX(Mask Index Attribute:NiagaraInt32=$Particles.SpawnGroup, Mask Index FL=2, Mask Index FR=3, Mask Index RL=0, Mask Index RR=1)
   /SetVariables(Particles.RibbonFacing=&MakeVector(
     X=0
     Y=&Sine(
      Normalized Angle=&Subtract_Float(A=$Emitter.NormalizedLoopAge, B=$Particles.NormalizedAge)
      Period=1
      Scale=&FloatFromCurve(FloatCurve=DataInterface() {propertyValues {Curve {
          defaultValue=1
          keys [
           {arriveTangent=4.1785, interpMode=RCIM_Cubic, leaveTangent=4.1785, tangentMode=RCTM_User, time=0, value=0}
           {arriveTangent=0.0224, interpMode=RCIM_Cubic, leaveTangent=0.0224, tangentMode=RCTM_User, time=0.5, value=1}
          ]
         }}})
      Bias=0
     )
     Z=1
    )) {enabled=false, isSetParameters=true, module=SetVariables_#3}
   &InheritSourceMovement(
    Applied Position Scale=&Multiply_VectorByFloat(Vector=<1 1 1>, Float=1)
    ApplyForce=false
    ApplyPosition=true
    ApplyPositionFalloff=false
   )
   &ParticleState()
   &SolveForcesAndVelocity(Clamp Velocity=false, Limit Acceleration=false)
   &ScaleColor(ScaleA=true, ScaleRGB=true)
   &VC_UpdateRibbon_v1(Detach Locked Particle=true, Enable Skeletal Position=true)
   /SetVariables(
    Particles.BackGround03:LinearColor=&MultiplyLinearColor(
     Linear Color=<rgba 0.831 0.5334 0 55>
     Scale Factor=&MakeVector4FromVectorAndFloat(
      "Vector (RGB)"=<1 1 1>
      "Float (Alpha)"=&FloatFromCurve(FloatCurve=DataInterface() {propertyValues {Curve {
          defaultValue=-0.9281
          keys [<at 0 0 RCIM_Cubic>, {arriveTangent=0.0361, interpMode=RCIM_Cubic, leaveTangent=0.0361, tangentMode=RCTM_User, time=0.8, value=1}]
         }}})
     )
    )
    Particles.BlackColor:LinearColor=&MultiplyLinearColor(
     Linear Color=<rgba 0 0 0 1>
     Scale Factor=&MakeVector4FromVectorAndFloat(
      "Vector (RGB)"=<1 1 1>
      "Float (Alpha)"=&FloatFromCurve(FloatCurve=DataInterface() {propertyValues {Curve {keys [
           <at 0 -1 RCIM_Cubic>
           {arriveTangent=1.1995, interpMode=RCIM_Cubic, leaveTangent=1.1995, tangentMode=RCTM_User, time=0.2, value=4}
           {arriveTangent=0.074, interpMode=RCIM_Cubic, leaveTangent=0.074, tangentMode=RCTM_User, time=0.81, value=5}
           <at 1 -1.5 RCIM_Cubic>
          ]}}})
     )
    )
   ) {isSetParameters=true, module=SetVariables_#4}
   /SetVariables(Particles.Position=&AddVectorToPosition(
     Position=$Particles.Position
     Vector=&TransformVector(
      Vector=&MakeVector(
       X=0
       Y=0
       Z=&Sine(
        Normalized Angle=&Add_Float(
         A=&Multiply_Float(A=$Particles.NormalizedAge, B=-1)
         B=&Multiply_Float(A=$Emitter.LoopedAge, B=1)
        )
        Period=1
        Scale=&FloatFromCurve(FloatCurve=DataInterface() {propertyValues {Curve {keys [<at 0.1 0>,<at 0.45 0.85>]}}}, Scale Curve=5)
        Bias=0
       )
      )
      Destination Space=::Simulation
      Apply Scale=true
     )
    )) {isSetParameters=true, module=SetVariables_#5}
   /SetVariables(Particles.Position=&AddVectorToPosition(
     Position=$Particles.Position
     Vector=&VC_ConvertVectorToWorld(Vector:Vector3f=&MakeVector(
       X=0
       Y=0
       Z=&FloatFromCurve(
        FloatCurve=DataInterface() {propertyValues {Curve {keys [
            <at 0 0 RCIM_Cubic>
            {arriveTangent=5.5287, interpMode=RCIM_Cubic, leaveTangent=5.5287, tangentMode=RCTM_User, time=0.3, value=10}
            {arriveTangent=3.4214, interpMode=RCIM_Cubic, leaveTangent=3.4214, tangentMode=RCTM_User, time=0.5, value=12}
            {arriveTangent=-3.7582, interpMode=RCIM_Cubic, leaveTangent=-3.7582, tangentMode=RCTM_User, time=1, value=5}
           ]}}}
        Scale Curve=0.3
       )
      ))
    )) {isSetParameters=true, module=SetVariables_#6}
   /SetVariables(Particles.RibbonFacing=&MaskVectorBySpawnGroup(
     Masked Vector=&MakeVector(
      X=1
      Y=0
      Z=&Sine(
       Normalized Angle=&Add_Float(
        A=&Multiply_Float(A=$Particles.NormalizedAge, B=1)
        B=&Multiply_Float(A=$Emitter.LoopedAge, B=1)
       )
       Period=1
       Scale=&FloatFromCurve(FloatCurve=DataInterface() {propertyValues {Curve {
           defaultValue=1
           keys [
            {arriveTangent=4.1785, interpMode=RCIM_Cubic, leaveTangent=4.1785, tangentMode=RCTM_User, time=0, value=0}
            {arriveTangent=0.0224, interpMode=RCIM_Cubic, leaveTangent=0.0224, tangentMode=RCTM_User, time=0.5, value=1}
           ]
          }}})
       Bias=0
      )
     )
     Passthrough Vector=&MakeVector(
      X=1
      Y=0
      Z=&Sine(
       Normalized Angle=&Add_Float(
        A=&Multiply_Float(A=$Particles.NormalizedAge, B=1)
        B=&Multiply_Float(A=$Emitter.LoopedAge, B=1)
       )
       Period=1
       Scale=&FloatFromCurve(FloatCurve=DataInterface() {propertyValues {Curve {
           defaultValue=1
           keys [
            {arriveTangent=4.1785, interpMode=RCIM_Cubic, leaveTangent=4.1785, tangentMode=RCTM_User, time=0, value=0}
            {arriveTangent=0.0224, interpMode=RCIM_Cubic, leaveTangent=0.0224, tangentMode=RCTM_User, time=0.5, value=1}
           ]
          }}})
       Bias=0
      )
     )
    )) {isSetParameters=true, module=SetVariables_#7}
   &ScaleRibbonWidth(Ribbon Width Scale=&SelectFloatFromArray(Direct Array Index=&SpawnGroup(), Float Selection Array=DataInterface() {propertyValues {FloatData [1,-1]}}))
   /SetVariables(
    Particles.R2Width:NiagaraFloat=&SelectFloatFromArray(Direct Array Index=&SpawnGroup(), Float Selection Array=DataInterface() {propertyValues {FloatData [0,-10]}})
    Particles.R3Width:NiagaraFloat=&SelectFloatFromArray(Direct Array Index=&SpawnGroup(), Float Selection Array=DataInterface() {propertyValues {FloatData [10,0]}})
   ) {isSetParameters=true}
  ]
  rendererClasses [&NiagaraRibbonRendererProperties]
  renderers [
   {rendererClass=&NiagaraRibbonRendererProperties}
   {rendererClass=&NiagaraRibbonRendererProperties, rendererIndex=1}
   {rendererClass=&NiagaraRibbonRendererProperties, rendererIndex=2}
  ]
 }
 NE_WheelWrap {
  emitterUpdateScript [
   &EmitterState(
    Inactive Response=::NewEnumerator1
    Life Cycle Mode=::NewEnumerator1
    Loop Behavior=::NewEnumerator1
    Loop Duration Mode=::NewEnumerator1
    UseLoopDelay=false
   )
   &SpawnBurst_Instantaneous(Spawn Count=2, Use Spawn Probability=false, UseLoopCountLimit=false)
  ]
  particleSpawnScript [
   &VC_PaintedVariantsOverride(
    Black=<rgba 0 0 0 1>
    BurntSienna=<rgba 0.3333 0.1029 0.0411 1>
    Cobalt=<rgba 0.0117 0.0332 1.1719 1>
    Crimson=<rgba 0.75 0.0547 0.0619 1>
    DefaultColor=<rgba 0.8308 0.5333 0 1>
    ForestGreen=<rgba 0 0.5 0.0253 1>
    Grey=<rgba 0.3125 0.3125 0.3125 1>
    Lime=<rgba 0.3828 0.7656 0 1>
    Orange=<rgba 1 0.2059 0 1>
    Pink=<rgba 0.7865 0.172 0.563 1>
    Purple=<rgba 0.1452 0.0508 0.8125 1>
    Saffron=<rgba 0.8542 0.6725 0.0756 1>
    SkyBlue=<rgba 0.0469 0.397 1 1>
    TitaniumWhite=<rgba 0.7083 0.7083 0.7083 1>
   ) {enabled=false}
   &InitializeParticle(
    Color Mode=::NewEnumerator0
    Lifetime=$Emitter.CurrentLoopDuration
    Mesh Scale Mode=::NewEnumerator4
    Sprite Rotation Mode=::NewEnumerator2
    Sprite Size Mode=::NewEnumerator3
    UsePositionOffset=false
    Write Material Random=true
   )
   &PartitionParticles()
   &InitialMeshOrientation(
    Mesh Orientation Mode=::NewEnumerator7
    Rotation=&SelectVectorFromArray(
     Direct Array Index=$Particles.PartitionParticles.Partition
     Vector Selection Array=DataInterface() {propertyValues {FloatData [<0 0 0>, <-0.5 0 0>]}}
    )
    Use Rotation Vector=true
   )
   &SkeletalMeshLocation(
    Bone / Socket Index=&Add_Integer(A=$Particles.PartitionParticles.Partition, B=4)
    Bone / Socket Sampling Mode=::NewEnumerator5
    Interpolate Mesh Sampling=false
    MaskSpawnGroup=false
    Mesh Sampling Type=::NewEnumerator6
    Skeletal Mesh=$System.SkeletalMesh
   )
  ]
  particleUpdateScript [
   &FN_DisableTireFX(
    Mask Index Attribute:NiagaraInt32=$Particles.PartitionParticles.Partition
    Mask Index FL=2
    Mask Index FR=3
    Mask Index RL=0
    Mask Index RR=1
   )
   &VC_WheelWrapScale(Wheel Index Attribute:NiagaraInt32=&Add_Integer(A=$Particles.PartitionParticles.Partition, B=2))
   &ParticleState(Kill Particles When Lifetime Has Elapsed=false, Let Infinitely Lived Particles Die When Emitter Deactivates=true)
   &UpdateMeshOrientation(Rotation Coordinate Space=::NewEnumerator0, Rotation Vector=<0 0.8 0>)
   /SetVariables(Particles.BackGround:LinearColor=<rgba 0.8308 0.5333 0 1>, Particles.BlackColor:LinearColor=<rgba 0 0 0 1>) {isSetParameters=true}
   &SkeletalMeshLocation(
    Bone / Socket Index=&Add_Integer(A=$Particles.PartitionParticles.Partition, B=4)
    Bone / Socket Sampling Mode=::NewEnumerator5
    MaskSpawnGroup=false
    Mesh Sampling Type=::NewEnumerator6
    Orientation Sampling=::NewEnumerator1
    Sampled Position Offset=&SelectVectorFromArray(
     Direct Array Index=$Particles.PartitionParticles.Partition
     Vector Selection Array=DataInterface() {propertyValues {FloatData [<0 0 0>, <0 0 0>]}}
    )
    Skeletal Mesh=$System.SkeletalMesh
   ) {module=SkeletalMeshLocation001}
   &SolveForcesAndVelocity(Clamp Velocity=false, Limit Acceleration=false)
  ]
  rendererClasses [&NiagaraMeshRendererProperties]
  renderers [
   {rendererClass=&NiagaraMeshRendererProperties, rendererIndex=0}
   {rendererClass=&NiagaraMeshRendererProperties, rendererIndex=1}
  ]
 }
}
summary {
 systemName=NS_DriftTrail_CrimeTape1
 emitters [
  {emitterName=NE_WheelWrap, rendererClasses [&NiagaraMeshRendererProperties]}
  {emitterName=NE_R2_DoNot001, rendererClasses [&NiagaraRibbonRendererProperties]}
  {emitterName=NE_R3_Tape001, rendererClasses [&NiagaraRibbonRendererProperties]}
  {emitterName=NE_R1_Warning001, rendererClasses [&NiagaraRibbonRendererProperties]}
 ]
 userVariables [
  User.bIsFrontendPreview {type=NiagaraBool}
  User.bIsPainted {type=NiagaraBool}
  User.ColorVariantID {type=NiagaraInt32}
  User.DTCarOffset {type=NiagaraPosition, description="", defaultValue {}}
  User.GameModeID {type=NiagaraInt32}
  User.offset {type=Vector3f, description="offset from wheels(set through cosmetic template)", defaultValue {struct=Vector3f, value=<0 0 0>}}
  User.PaintedColor {type=LinearColor, description="", defaultValue {struct=LinearColor, value=<rgba 1 1 1 1>}}
  User.PaintedScalar {type=NiagaraFloat, defaultValue=1, description=""}
  User.VehicleFlags {type=NiagaraInt32, defaultValue=806289408}
  User.Wheel_Radius_Array {
   type=NiagaraDataInterfaceArrayFloat
   defaultValue {
    struct=NiagaraExt_VariableValue_DataInterface
    value {
     dataInterface=&self:NiagaraDataInterfaceArrayFloat_1
     dataInterfaceClass=&NiagaraDataInterfaceArrayFloat
    }
   }
  }
  User.Wheel_Width_Array {
   type=NiagaraDataInterfaceArrayFloat
   defaultValue {
    struct=NiagaraExt_VariableValue_DataInterface
    value {
     dataInterface=&self:NiagaraDataInterfaceArrayFloat_0
     dataInterfaceClass=&NiagaraDataInterfaceArrayFloat
    }
   }
  }
  User.WheelTouchingArray {
   type=NiagaraDataInterfaceArrayFloat
   defaultValue {
    struct=NiagaraExt_VariableValue_DataInterface
    value {
     dataInterface=&self:NiagaraDataInterfaceArrayFloat_2
     dataInterfaceClass=&NiagaraDataInterfaceArrayFloat
    }
   }
  }
 ]
}
